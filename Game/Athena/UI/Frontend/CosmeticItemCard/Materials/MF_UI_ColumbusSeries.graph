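# UI material function that paints the Columbus series background for cosmetic item cards when bIsColumbusSeries is on, otherwise passing its input through.
# A smoothed, rotated two-tone blue gradient is subtly darkened by the SW2024 pattern texture, then blended toward bright blue by an offset exponential radial glow.
# UVs come from MF_UI_UVScaling's fill scaling.
# Panning radial-splatter, screen-aligned starfield and SeriesTexture branches remain in the graph but do not reach the output.
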MaterialFunction /Game/Athena/UI/Frontend/CosmeticItemCard/Materials/MF_UI_ColumbusSeries
graph {nodes [
  n1 = Constant3Vector() {constant=<rgba 0 0.3667 0.9774 1>}
  n2 = MaterialFunctionCall(UVs=NamedRerouteUsage(), Texture Scale=Constant() {r=1}) {materialFunction=&ScaleUVsByCenter}
  n3 = AppendVector(A=Constant() {r=-0.5}, B=Constant() {r=0.25})
  n4 = Multiply(A=MaterialFunctionCall(UVs=Add(A=@n2.UVs, B=@n3), Radius=Constant() {r=0.75}, Density=Constant() {r=1.2}) {materialFunction=&RadialGradientExponential}.RadialGradientExponential) {constB=0.3}
  Add(A=@n1, B=@n4)
  ComponentMask() {g=true, r=true}
  FunctionInput(Preview=Constant3Vector() {constant=<rgba 0 0.9473 0.9144 1>}) {inputName=RGB 1, sortPriority=1}
  FunctionInput(Preview=Constant3Vector() {constant=<rgba 0 0.1207 0.5573 1>}) {inputName=RGB 2, sortPriority=1}
  n5 = MaterialFunctionCall(UVs=NamedRerouteUsage(), GradientStart=Constant(), GradientEnd=Constant() {r=1}, RotationAngle=Constant() {r=0.28}) {materialFunction=&MF_UI_LinearGradient}
  n6 = TextureSample(UVs=MaterialFunctionCall(UVs=NamedRerouteUsage(), Texture Scale=Constant() {r=1.7}) {materialFunction=&ScaleUVsByCenter}.UVs) {samplerType=SAMPLERTYPE_Alpha, texture=/Game/UI/Frontend/ItemShop/Art/SW2024/T_UI_SWPattern}
  n7 = LinearInterpolate(
   A=LinearInterpolate(
    A=Constant3Vector() {constant=<rgba 0.0042 0.0261 0.1719 1>}
    B=Constant3Vector() {constant=<rgba 0.013 0.0953 0.6038 1>}
    Alpha=SmoothStep(Value=@n5.Result)
   )
   Alpha=Multiply(A=OneMinus(None=@n6.R)) {constB=0.12}
  ) {constB=0}
  FunctionOutput(None=StaticSwitchParameter(True=LinearInterpolate(A=@n7, B=@n1, Alpha=@n4), False=FunctionInput()) {parameterName=bIsColumbusSeries})
  n8 = MaterialFunctionCall() {materialFunction=&VectorToRadialValue}
  n9 = Panner(
   Coordinate=@n8.Radial Coordinates
   Speed=AppendVector(A=ScalarParameter() {defaultValue=0.0125, parameterName=U-Speed}, B=ScalarParameter() {defaultValue=-2, parameterName=V-Speed})
  )
  n10 = Multiply(
   A=@n9
   B=AppendVector(A=ScalarParameter() {defaultValue=1, parameterName=R-U-Tiles}, B=ScalarParameter() {defaultValue=0.01, parameterName=R-V-Tiles})
  )
  n11 = Multiply(
   A=@n9
   B=AppendVector(A=ScalarParameter() {defaultValue=3, parameterName=G-U-Tiles}, B=ScalarParameter() {defaultValue=0.03, parameterName=G-V-Tiles})
  )
  n12 = Add(
   A=Multiply(A=TextureSample(UVs=@n10) {texture=/Game/Global/Textures/Default/PointSplatter/T_Ui_ColumbusSeries_512}.RGB) {constB=2}
   B=Multiply(A=TextureSample(UVs=@n11) {texture=&T_Ui_ColumbusSeries_512}.RGB) {constB=2}
  )
  n13 = Power(Base=@n8.Linear Distance, Exp=ScalarParameter() {defaultValue=2, parameterName=MaskPower})
  n14 = Multiply(A=TextureCoordinate(), B=ScalarParameter() {parameterName=R-Local-UV-Influence})
  n15 = MaterialFunctionCall(Float2=MaterialFunctionCall() {materialFunction=&ScreenResolution}.Visible Resolution) {materialFunction=&BreakOutFloat2Components}
  n16 = Divide(A=@n15.R, B=Max(A=@n15.G) {constB=0.1})
  n17 = Constant() {r=1}
  n18 = Multiply(
   A=Multiply(
    A=AppendVector(A=@n16, B=@n17)
    B=MaterialFunctionCall() {materialFunction=&ScreenAlignedUVs}."X 100%, Y 100%"
   )
   B=ScalarParameter() {defaultValue=1, parameterName=Global Tile}
  )
  n19 = Multiply(
   A=TextureSample(UVs=Add(A=@n14, B=@n18)) {texture=/Game/Global/Textures/Default/PointSplatter/T_UI_Starfield}.RGB
   B=ScalarParameter() {parameterName=R-Strength}
  ) {constB=2}
  n20 = Multiply(A=TextureCoordinate(), B=ScalarParameter() {parameterName=G-Local-UV-Influence})
  n21 = Multiply(
   A=TextureSample(UVs=Add(A=@n18, B=@n20)) {texture=&T_UI_Starfield}.RGB
   B=ScalarParameter() {parameterName=G-Strength}
  ) {constB=2}
  n22 = MaterialFunctionCall(Float2=TextureCoordinate() {coordinateIndex=3}) {materialFunction=&BreakOutFloat2Components}
  n23 = Divide(A=@n22.R, B=Max(A=@n22.G))
  n24 = Multiply(A=AppendVector(A=@n23, B=Constant() {r=1}), B=TextureCoordinate() {coordinateIndex=4})
  n25 = AppendVector(A=Multiply(A=OneMinus(None=@n23)) {constB=0.5}, B=Constant())
  LinearInterpolate(
   A=LinearInterpolate(A=Multiply(A=@n13, B=@n12), B=Add(A=@n19, B=@n21), Alpha=ScalarParameter() {defaultValue=1, parameterName=bIsStatic})
   B=TextureSampleParameter2D(UVs=Add(A=@n24, B=@n25)) {group=Series, parameterName=SeriesTexture, texture=/Game/Global/Textures/Default/Masks_and_Gradients/T_Ui_LavaSeries_Bg}.RGB
   Alpha=ScalarParameter() {parameterName=Background}
  )
  MaterialFunctionCall!(ScreenUV, TextureSize, ScalePivotPoint) {materialFunction=/FortUILibrary/MaterialFunctions/Basics/MF_UI_ScreenSpaceUVScaling}
  Multiply(A=@n4, B=@n1)
  NamedRerouteDeclaration(None=MaterialFunctionCall() {materialFunction=&MF_UI_UVScaling}.Scaled To Fill) {name=UVFit, nodeColor=<rgba 1 0.8672 0 1>}
 ]}
