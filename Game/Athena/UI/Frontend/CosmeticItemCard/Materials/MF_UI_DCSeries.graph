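# Material function for the DC series background on frontend cosmetic item cards, switched by bIsDCUSeries with a pass-through input otherwise.
# Light rays come from counter-panning polar noise fed into another noise lookup, masked by screen-aligned panning noise inside an offset, rotatable radial gradient.
# Two RGB inputs tint the rays and a vignetted metal-noise and scratch layer.
# The Background parameter blends this composite toward an aspect-corrected SeriesTexture, which fully replaces it at the default.
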MaterialFunction /Game/Athena/UI/Frontend/CosmeticItemCard/Materials/MF_UI_DCSeries
graph {nodes [
  n1 = MaterialFunctionCall(Float2=MaterialFunctionCall() {materialFunction=&ScreenResolution}.Visible Resolution) {materialFunction=&BreakOutFloat2Components}
  n2 = AppendVector(A=Divide(A=@n1.R, B=Max(A=@n1.G)), B=Constant() {r=1})
  n3 = MaterialFunctionCall() {materialFunction=&ScreenAlignedUVs}
  Add(
   A=Multiply(B=ScalarParameter() {parameterName=Local-UV})
   B=Multiply(A=Multiply(A=@n2, B=@n3."X 100%, Y 100%"), B=ScalarParameter() {parameterName=Global-Logo-Tile})
  )
  n4 = MaterialFunctionCall() {materialFunction=&GetUserInterfaceUV}
  n5 = MaterialFunctionCall(Float2=@n4.Pixel Size) {materialFunction=&BreakOutFloat2Components}
  n6 = Max(A=@n5.G)
  n7 = Multiply(A=AppendVector(A=Divide(A=@n5.R, B=@n6), B=Constant() {r=1}), B=@n4.Normalized UV)
  n8 = VectorParameter() {group=DCU-Series, parameterName=Light-Ray-Offset}
  n9 = MaterialFunctionCall(Vector or UVs=Add(
    A=Subtract(A=@n7) {constB=0.5}
    B=MaterialFunctionCall(X=@n8.R, Y=@n8.R) {materialFunction=&MakeFloat2}.Result
   )) {materialFunction=&VectorToRadialValue}
  n10 = ScalarParameter() {defaultValue=0.25, group=DCU-Series, parameterName=Light-Rays-Time}
  n11 = TextureSample(UVs=Panner(Coordinate=@n9.Vector Converted to Angle, Time=@n10) {speedX=0.125}) {samplerType=SAMPLERTYPE_Grayscale, texture=&LowResBlurredNoise}
  n12 = TextureSample(UVs=Panner(Coordinate=@n9.Vector Converted to Angle, Time=@n10) {speedX=-0.125}) {samplerType=SAMPLERTYPE_Grayscale, texture=&LowResBlurredNoise}
  n13 = MaterialFunctionCall(Float2=MaterialFunctionCall() {materialFunction=&ScreenResolution}.Visible Resolution) {materialFunction=&BreakOutFloat2Components}
  n14 = Max(A=@n13.G)
  n15 = Multiply(
   A=AppendVector(A=Divide(A=@n13.R, B=@n14), B=Constant() {r=1})
   B=MaterialFunctionCall() {materialFunction=&ScreenAlignedUVs}."X 100%, Y 100%"
  )
  n16 = ScalarParameter() {defaultValue=4, group=DCU-Series, parameterName=Global UV Foreground}
  n17 = MaterialFunctionCall() {materialFunction=&GetUserInterfaceUV}
  n18 = MaterialFunctionCall(Float2=@n17.Pixel Size) {materialFunction=&BreakOutFloat2Components}
  n19 = Max(A=@n18.G)
  n20 = Multiply(A=AppendVector(A=Divide(A=@n18.R, B=@n19), B=Constant() {r=1}), B=@n17.Normalized UV)
  n21 = ScalarParameter() {group=DCU-Series, parameterName=Rotation}
  n22 = ScalarParameter() {defaultValue=2, group=DCU-Series, parameterName=Lighting-U-Tile}
  n23 = ScalarParameter() {defaultValue=0.5, group=DCU-Series, parameterName=Lighting-V-Tile}
  n24 = Add(
   A=Multiply(A=Rotator(Coordinate=@n20, Time=@n21) {centerX=0.1, speed=6.2832}, B=AppendVector(A=@n22, B=@n23))
   B=AppendVector(
    A=ScalarParameter() {defaultValue=2.0107, group=DCU-Series, parameterName=Light-U-Offset}
    B=ScalarParameter() {defaultValue=-0.081, group=DCU-Series, parameterName=Light-Y-Offset}
   )
  )
  n25 = ScalarParameter() {defaultValue=0.25, group=DCU-Series, parameterName=Localized UV}
  n26 = Multiply(A=Time(), B=ScalarParameter() {defaultValue=0.25, group=DCU-Series, parameterName=Time Factor})
  n27 = Panner(Coordinate=Add(A=Multiply(A=@n15, B=@n16), B=Multiply(A=@n24, B=@n25)), Time=@n26) {speedX=0.35, speedY=0.125}
  n28 = ScalarParameter() {defaultValue=8, group=DCU-Series, parameterName=Global UV Background}
  n29 = Panner(Coordinate=Add(A=Multiply(A=@n15, B=@n28)), Time=@n26) {speedX=0.25, speedY=0.25}
  n30 = LinearInterpolate(
   B=Add(
    A=TextureSample(UVs=@n27) {samplerType=SAMPLERTYPE_Grayscale, texture=&LowResBlurredNoise}.R
    B=TextureSample(UVs=@n29) {samplerType=SAMPLERTYPE_Grayscale, texture=&LowResBlurredNoise}.R
   )
   Alpha=MaterialFunctionCall(
    UVs=@n24
    CenterPosition=Constant2Vector() {g=0.5, r=0.5}
    Radius=ScalarParameter() {defaultValue=0.5, group=DCU-Series, parameterName=Radius-Light}
    Density=ScalarParameter() {defaultValue=0.125, group=DCU-Series, parameterName=Density Light}
   ) {materialFunction=&RadialGradientExponential}.RadialGradientExponential
  )
  n31 = LinearInterpolate(
   B=TextureSample(UVs=Add(A=@n11.R, B=@n12.R)) {samplerType=SAMPLERTYPE_Grayscale, texture=&LowResBlurredNoise}.RGB
   Alpha=@n30
  )
  n32 = Power(Base=@n31, Exp=ScalarParameter() {group=DCU-Series, parameterName=Light-Rays-Power})
  n33 = FunctionInput(Preview=Constant3Vector() {constant=<rgba 0.0228 0.0254 0.0365 0>}) {inputName=RGB 1, sortPriority=1}
  n34 = Multiply(A=@n31, B=@n33)
  n35 = MaterialFunctionCall(
   CenterPosition=Constant2Vector() {g=0.5, r=0.5}
   Radius=ScalarParameter() {defaultValue=1, group=DCU-Series, parameterName=Radius-Enhanced}
   Density=ScalarParameter() {defaultValue=1.5, group=DCU-Series, parameterName=Density-Enhanced}
  ) {materialFunction=&RadialGradientExponential}
  n36 = LinearInterpolate(
   A=@n33
   B=FunctionInput(Preview=Constant3Vector() {constant=<rgba 0.0079 0.0084 0.0104 0>}) {inputName=RGB 2, sortPriority=1}
   Alpha=@n35.RadialGradientExponential
  )
  n37 = MaterialFunctionCall() {materialFunction=&GetUserInterfaceUV}
  n38 = MaterialFunctionCall(Float2=@n37.Pixel Size) {materialFunction=&BreakOutFloat2Components}
  n39 = AppendVector(A=Divide(A=@n38.R, B=Max(A=@n38.G)), B=Constant() {r=1})
  n40 = LinearInterpolate(
   B=@n36
   Alpha=TextureSample(UVs=Multiply(A=@n39, B=@n37.Normalized UV)) {samplerType=SAMPLERTYPE_Masks, texture=/Game/Global/Textures/Default/Noise/T_Metal_Noise}.G
  )
  n41 = Multiply(A=@n36, B=ScalarParameter() {defaultValue=1, group=DCU-Series, parameterName=Scratch-Hotspot})
  n42 = TextureCoordinate()
  n43 = AppendVector(
   A=ScalarParameter() {defaultValue=1, group=DCU-Series, parameterName=Scratches-U-Tile}
   B=ScalarParameter() {defaultValue=1, group=DCU-Series, parameterName=Scratches-Y-Tile}
  )
  n44 = Power(
   Base=TextureSample(UVs=Multiply(A=@n42, B=@n43)) {samplerType=SAMPLERTYPE_Masks, texture=/Game/Global/Textures/Default/Noise/T_Pattern_Noise_M}.G
   Exp=ScalarParameter() {defaultValue=4, group=DCU-Series, parameterName=Noise-Power}
  )
  n45 = Add(
   A=Add(A=Add(A=@n32, B=@n34), B=Multiply(A=@n33, B=@n30))
   B=LinearInterpolate(A=LinearInterpolate(A=@n40, B=@n41, Alpha=@n44), B=Constant3Vector(), Alpha=OneMinus(None=@n35.RadialGradientExponential))
  )
  n46 = MaterialFunctionCall(Float2=TextureCoordinate() {coordinateIndex=3}) {materialFunction=&BreakOutFloat2Components}
  n47 = Divide(A=@n46.R, B=Max(A=@n46.G))
  n48 = AppendVector(A=@n47, B=Constant() {r=1})
  n49 = TextureCoordinate() {coordinateIndex=4}
  n50 = Multiply(A=OneMinus(None=@n47)) {constB=0.5}
  n51 = Constant()
  n52 = TextureSampleParameter2D(UVs=Add(A=Multiply(A=@n48, B=@n49), B=AppendVector(A=@n50, B=@n51))) {group=Series, parameterName=SeriesTexture, texture=/Game/Global/Textures/Default/Masks_and_Gradients/T_Ui_LavaSeries_Bg}
  n53 = ScalarParameter() {defaultValue=1, parameterName=Background}
  FunctionOutput(None=StaticSwitchParameter(True=LinearInterpolate(A=@n45, B=@n52.RGB, Alpha=@n53), False=FunctionInput()) {parameterName=bIsDCUSeries})
  TextureCoordinate()
  TextureCoordinate()
  TextureCoordinate()
  TextureCoordinate()
 ]}
