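# Material function for the default background of frontend cosmetic item cards, blending two input colors (RGB 2 to RGB 1) through a patterned radial mask.
# A rotated, scaled triangle-grid texture, aspect-corrected using texcoord 3 over texcoord 4 UVs, is mixed with an exponential radial gradient.
# That mix, multiplied by the gradient again and saturated, drives the color blend.
# An additive "Enhanced" gradient is wired in but zero-weighted by a constant.
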MaterialFunction /Game/Athena/UI/Frontend/CosmeticItemCard/Materials/MF_UI_Default
graph {nodes [
  n1 = Multiply(B=MaterialFunctionCall(
    CenterPosition=Constant2Vector() {r=1}
    Radius=ScalarParameter() {defaultValue=1.25, group=Default-Background, parameterName=Radius-Enhanced}
    Density=ScalarParameter() {defaultValue=1.5, group=Default-Background, parameterName=Density-Enhanced}
   ) {materialFunction=&RadialGradientExponential}.RadialGradientExponential) {constA=1}
  n2 = ScalarParameter() {defaultValue=0.05, group=Default-Background, parameterName=Triangle - U}
  n3 = ScalarParameter() {defaultValue=0.5, group=Default-Background, parameterName=Triangle - Y}
  n4 = MaterialFunctionCall(Float2=TextureCoordinate() {coordinateIndex=3}) {materialFunction=&BreakOutFloat2Components}
  n5 = Divide(A=@n4.R, B=Max(A=@n4.G))
  n6 = Multiply(A=AppendVector(A=@n5, B=Constant() {r=1}), B=TextureCoordinate() {coordinateIndex=4})
  n7 = AppendVector(A=Multiply(A=OneMinus(None=@n5)) {constB=0.5}, B=Constant())
  n8 = Rotator(
   Coordinate=Add(A=AppendVector(A=@n2, B=@n3), B=Add(A=@n6, B=@n7))
   Time=ScalarParameter() {defaultValue=-0.025, group=Default-Background, parameterName=AngleTriangle, sortPriority=1}
  ) {speed=6.2832}
  n9 = ScalarParameter() {defaultValue=0.5, group=Default-Background, parameterName=TriangleScale, sortPriority=1}
  n10 = MaterialFunctionCall(
   CenterPosition=AppendVector(
    A=ScalarParameter() {defaultValue=0.5, group=Default-Background, parameterName=RadialMask-U}
    B=ScalarParameter() {defaultValue=1, group=Default-Background, parameterName=RadialMask-Y}
   )
   Radius=ScalarParameter() {defaultValue=2, group=Default-Background, parameterName=Radius, sortPriority=1}
   Density=ScalarParameter() {defaultValue=1.5, group=Default-Background, parameterName=Density, sortPriority=1}
  ) {materialFunction=&RadialGradientExponential}
  n11 = LinearInterpolate(
   A=TextureSample(UVs=Multiply(A=@n8, B=@n9)) {texture=/Game/Global/Textures/Default/Masks_and_Gradients/T-TriangleGrid}.R
   B=@n10.RadialGradientExponential
   Alpha=ScalarParameter() {defaultValue=0.85, group=Default-Background, parameterName=Triangle, sortPriority=1}
  ) {constAlpha=0.8, constB=2}
  n12 = LinearInterpolate(
   A=FunctionInput(Preview=Constant3Vector() {constant=<rgba 0.0027 0.058 0.462 1>}) {inputName=RGB 2}
   B=FunctionInput(Preview=Constant3Vector() {constant=<rgba 0 0.4667 1 1>}) {inputName=RGB 1}
   Alpha=Saturate(None=Multiply(A=@n11, B=@n10.RadialGradientExponential))
  )
  FunctionOutput(None=LinearInterpolate(A=@n12, B=Add(A=@n1, B=@n12), Alpha=Constant()))
 ]}
