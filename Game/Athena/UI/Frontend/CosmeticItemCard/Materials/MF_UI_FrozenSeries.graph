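# UI material function for the Frozen Series cosmetic item card background, enabled by the bIsFrozenSeries static switch, otherwise passing an input through.
# Two panning ice-noise samples, mixing aspect-corrected screen UVs with local UVs, distort an exponential radial gradient that blends two frost colors.
# The Background parameter blends that frost toward SeriesTexture, aspect-fitted via UV channels 3 and 4; its default shows only the texture.
# The RGB 2 input is unconnected.
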MaterialFunction /Game/Athena/UI/Frontend/CosmeticItemCard/Materials/MF_UI_FrozenSeries
graph {nodes [
  n1 = MaterialFunctionCall(Float2=MaterialFunctionCall() {materialFunction=&ScreenResolution}.Visible Resolution) {materialFunction=&BreakOutFloat2Components}
  n2 = Max(A=@n1.G)
  n3 = Multiply(
   A=AppendVector(A=Divide(A=@n1.R, B=@n2), B=Constant() {r=1})
   B=MaterialFunctionCall() {materialFunction=&ScreenAlignedUVs}."X 100%, Y 100%"
  )
  n4 = Multiply(A=@n3, B=ScalarParameter() {defaultValue=8, group=Forzen-Series, parameterName=Global UV Background})
  n5 = Multiply(A=TextureCoordinate(), B=ScalarParameter() {defaultValue=0.25, group=Forzen-Series, parameterName=Localized UV})
  Add(A=@n4, B=@n5)
  FunctionInput(Preview=Constant3Vector() {constant=<rgba 0.0027 0.058 0.462 1>}) {inputName=RGB 2, sortPriority=1}
  n6 = Constant3Vector() {constant=<rgba 0 0.4667 1 1>}
  n7 = Add(A=Multiply(A=@n3, B=ScalarParameter() {defaultValue=4, group=Forzen-Series, parameterName=Global UV Foreground}), B=@n5)
  n8 = Multiply(A=Time(), B=ScalarParameter() {parameterName=TimeFactor})
  n9 = Multiply(A=@n4, B=@n5)
  n10 = Add(
   A=TextureSample(UVs=Panner(Coordinate=@n7, Time=@n8) {speedX=0.125, speedY=0.25}) {texture=/Game/Global/Textures/Default/Noise/T_UI_Ice_Noise}.R
   B=TextureSample(UVs=Panner(Coordinate=@n9, Time=@n8) {speedX=0.125, speedY=0.25}) {texture=&T_UI_Ice_Noise}.R
  )
  n11 = ScalarParameter() {group=Forzen-Series, parameterName=StormDistortionMultiply}
  n12 = MaterialFunctionCall(
   UVs=Add(A=TextureCoordinate(), B=Multiply(A=@n10, B=@n11))
   CenterPosition=Constant2Vector() {g=0.5, r=0.5}
   Radius=ScalarParameter() {defaultValue=1.25, group=Forzen-Series, parameterName=BG-Radius-Enhanced}
   Density=ScalarParameter() {defaultValue=1.5, group=Forzen-Series, parameterName=BG-Density-Enhanced}
  ) {materialFunction=&RadialGradientExponential}
  n13 = ScalarParameter() {group=Forzen-Series, parameterName=FrostPower}
  n14 = Multiply(
   A=LinearInterpolate(
    A=VectorParameter() {group=Forzen-Series, parameterName=RGB-Mid, defaultValue=<rgba 0.1119 0.3515 0.5711 1>}.RGB
    B=FunctionInput(Preview=@n6) {inputName=RGB 1, sortPriority=1}
    Alpha=Power(Base=@n12.RadialGradientExponential, Exp=@n13)
   )
   B=ScalarParameter() {group=Forzen-Series, parameterName=FrostColor-Multiply}
  )
  n15 = MaterialFunctionCall(Float2=TextureCoordinate() {coordinateIndex=3}) {materialFunction=&BreakOutFloat2Components}
  n16 = Divide(A=@n15.R, B=Max(A=@n15.G))
  n17 = AppendVector(A=@n16, B=Constant() {r=1})
  n18 = TextureCoordinate() {coordinateIndex=4}
  n19 = Multiply(A=OneMinus(None=@n16)) {constB=0.5}
  n20 = Constant()
  n21 = TextureSampleParameter2D(UVs=Add(A=Multiply(A=@n17, B=@n18), B=AppendVector(A=@n19, B=@n20))) {group=Series, parameterName=SeriesTexture, texture=/Game/Global/Textures/Default/Masks_and_Gradients/T_Ui_LavaSeries_Bg}
  n22 = ScalarParameter() {defaultValue=1, parameterName=Background}
  FunctionOutput(None=StaticSwitchParameter(True=LinearInterpolate(A=@n14, B=@n21.RGB, Alpha=@n22), False=FunctionInput()) {parameterName=bIsFrozenSeries})
 ]}
