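# Material function for the Lava Series background on frontend cosmetic item cards, switched by bIsLavaSeries against a pass-through input.
# It adds embers and smoke, tinted between two RGB inputs, over an aspect-corrected series texture.
# Embers come from two panning dot-noise layers in screen-aligned UVs, the foreground blending in local UVs, masked by vertical falloff.
# Smoke is a radial gradient distorted by panning blurred noise.
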MaterialFunction /Game/Athena/UI/Frontend/CosmeticItemCard/Materials/MF_UI_LavaSeries
graph {nodes [
  n1 = MaterialFunctionCall(Float2=TextureCoordinate() {coordinateIndex=3}) {materialFunction=&BreakOutFloat2Components}
  n2 = Divide(A=@n1.R, B=Max(A=@n1.G))
  n3 = AppendVector(A=@n2, B=Constant() {r=1})
  n4 = TextureCoordinate() {coordinateIndex=4}
  n5 = Multiply(A=OneMinus(None=@n2)) {constB=0.5}
  n6 = Constant()
  n7 = TextureSampleParameter2D(UVs=Add(A=Multiply(A=@n3, B=@n4), B=AppendVector(A=@n5, B=@n6))) {group=Series, parameterName=SeriesTexture, texture=/Game/Global/Textures/Default/Masks_and_Gradients/T_Ui_LavaSeries_Bg}
  n8 = FunctionInput(Preview=Constant3Vector() {constant=<rgba 0.7157 0.0176 0.0044 1>}) {inputName=RGB 2}
  n9 = FunctionInput(Preview=Constant3Vector() {constant=<rgba 0.956 0.6172 0.0369 1>}) {inputName=RGB 1}
  n10 = MaterialFunctionCall(Float2=MaterialFunctionCall() {materialFunction=&ScreenResolution}.Visible Resolution) {materialFunction=&BreakOutFloat2Components}
  n11 = Max(A=@n10.G)
  n12 = Multiply(
   A=AppendVector(A=Divide(A=@n10.R, B=@n11), B=Constant() {r=1})
   B=MaterialFunctionCall() {materialFunction=&ScreenAlignedUVs}."X 100%, Y 100%"
  )
  n13 = ScalarParameter() {defaultValue=4, parameterName=Global UV Foreground}
  n14 = TextureCoordinate()
  n15 = ScalarParameter() {defaultValue=0.25, parameterName=Localized UV}
  n16 = Multiply(A=Time(), B=ScalarParameter() {defaultValue=0.25, parameterName=Time Factor})
  n17 = Panner(Coordinate=Add(A=Multiply(A=@n12, B=@n13), B=Multiply(A=@n14, B=@n15)), Time=@n16) {speedX=0.35, speedY=0.125}
  n18 = ScalarParameter() {parameterName=Ember-Scale}
  n19 = TextureSample(UVs=Multiply(A=@n17, B=@n18)) {texture=/Game/Global/Textures/Default/Noise/T_LockDeviceRandomDots}
  n20 = Add(A=Multiply(A=@n12, B=ScalarParameter() {defaultValue=8, parameterName=Global UV Background}))
  n21 = TextureSample(UVs=Multiply(A=@n18, B=Panner(Coordinate=@n20, Time=@n16) {speedX=0.25, speedY=0.25})) {texture=&T_LockDeviceRandomDots}
  n22 = Add(A=@n19.G, B=@n21.G) {constB=0}
  n23 = Multiply(A=LinearInterpolate(A=@n8, B=@n9, Alpha=@n22), B=Multiply(A=Add(A=@n19.R, B=@n21.R), B=@n22))
  n24 = ScalarParameter() {defaultValue=128, parameterName=Ember Intensity}
  n25 = TextureCoordinate()
  n26 = Multiply(
   A=Subtract(
    A=MaterialFunctionCall(Float2=@n25) {materialFunction=&BreakOutFloat2Components}.G
    B=ScalarParameter() {defaultValue=0.25, parameterName=Falloff-Subtract}
   )
   B=ScalarParameter() {defaultValue=4, parameterName=Falloff-Edge}
  )
  n27 = Add(
   A=TextureSample(UVs=@n17) {samplerType=SAMPLERTYPE_Grayscale, texture=&LowResBlurredNoise}.R
   B=TextureSample(UVs=@n17) {samplerType=SAMPLERTYPE_Grayscale, texture=&LowResBlurredNoise}.R
  )
  n28 = ScalarParameter() {defaultValue=0.25, parameterName=Smoke-Distortion-Intensity}
  n29 = MaterialFunctionCall(
   UVs=Add(A=TextureCoordinate(), B=Multiply(A=@n27, B=@n28))
   CenterPosition=AppendVector(A=ScalarParameter() {defaultValue=0.5, parameterName=Smoke-Offset}, B=ScalarParameter() {defaultValue=0.5, parameterName=Smoke-Offset})
   Radius=ScalarParameter() {defaultValue=0.5, group=Background, parameterName=Radius-Smoke, sortPriority=1}
   Density=ScalarParameter() {defaultValue=1.5, group=Background, parameterName=Density-Smoke, sortPriority=1}
  ) {materialFunction=&RadialGradientExponential}
  n30 = LinearInterpolate(A=@n8, B=@n9, Alpha=@n29.RadialGradientExponential)
  n31 = Add(
   A=LinearInterpolate(B=Multiply(A=@n23, B=@n24), Alpha=Saturate(None=@n26))
   B=Multiply(A=Multiply(A=@n29.RadialGradientExponential, B=@n30), B=ScalarParameter() {defaultValue=1, parameterName=Smoke-Intensity})
  )
  FunctionOutput(None=StaticSwitchParameter(True=Add(A=@n7.RGB, B=@n31), False=FunctionInput()) {parameterName=bIsLavaSeries})
 ]}
