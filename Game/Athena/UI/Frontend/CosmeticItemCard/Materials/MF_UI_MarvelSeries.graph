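# Material function that builds the animated Marvel-series background for cosmetic item cards in the frontend UI, gated by the bIsMarvelSeries static switch.
# Two channels of a noise texture pan as rotated, aspect-corrected streaks, with power and intensity modulated by a moving sine mask.
# Boosted streaks tinted by RGB 1 are added over a radial RGB 2-to-RGB 1 gradient.
# The Background scalar blends that result toward SeriesTexture.
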MaterialFunction /Game/Athena/UI/Frontend/CosmeticItemCard/Materials/MF_UI_MarvelSeries
graph {nodes [
  n1 = FunctionInput(Preview=Constant3Vector() {constant=<rgba 0 0.4667 1 1>}) {inputName=RGB 1, sortPriority=1}
  n2 = AppendVector(
   A=ScalarParameter() {defaultValue=1, group=Marvel-Series, parameterName=Streak-R-U-Tile}
   B=ScalarParameter() {defaultValue=1, group=Marvel-Series, parameterName=Streak-R-V-Tile}
  )
  n3 = MaterialFunctionCall() {materialFunction=&GetUserInterfaceUV}
  n4 = MaterialFunctionCall(Float2=@n3.Pixel Size) {materialFunction=&BreakOutFloat2Components}
  n5 = Max(A=@n4.G)
  n6 = Multiply(A=AppendVector(A=Divide(A=@n4.R, B=@n5), B=Constant() {r=1}), B=@n3.Normalized UV)
  n7 = Rotator(Coordinate=@n6, Time=ScalarParameter() {defaultValue=0.125, group=Marvel-Series, parameterName=Streak Angle}) {speed=6.2832}
  n8 = Multiply(A=Time(), B=ScalarParameter() {defaultValue=1, group=Marvel-Series, parameterName=Effect-Time-Factor})
  n9 = Add(
   A=Panner(Coordinate=Multiply(A=@n7, B=@n2), Time=@n8) {speedY=1}
   B=AppendVector(
    A=ScalarParameter() {defaultValue=1, group=Marvel-Series, parameterName=Streak-R-U-Offset}
    B=ScalarParameter() {defaultValue=1, group=Marvel-Series, parameterName=Streak-R-V-Offset}
   )
  )
  n10 = ScalarParameter() {group=Marvel-Series, parameterName=Streak-R-Power-Max}
  n11 = ScalarParameter() {group=Marvel-Series, parameterName=Streak-R-Power-Min}
  n12 = ScalarParameter() {defaultValue=0.125, group=Marvel-Series, parameterName=Streak Power Angle}
  n13 = ComponentMask(None=Panner(Coordinate=Rotator(Coordinate=@n6, Time=@n12) {speed=6.2832}, Time=@n8) {speedX=1}) {r=true}
  n14 = Saturate(None=Add(A=Sine(None=@n13), B=ScalarParameter() {group=Marvel-Series, parameterName=Sine Offset}))
  n15 = Multiply(
   A=Power(
    Base=TextureSample(UVs=@n9) {texture=/Game/Global/Textures/Default/Noise/T_Marvel_Series}.R
    Exp=LinearInterpolate(A=@n10, B=@n11, Alpha=@n14)
   )
   B=LinearInterpolate(
    A=ScalarParameter() {group=Marvel-Series, parameterName=Streak-R-Multiply-Max}
    B=ScalarParameter() {group=Marvel-Series, parameterName=Streak-R-Multiply-Min}
    Alpha=@n14
   )
  )
  n16 = AppendVector(
   A=ScalarParameter() {defaultValue=1, group=Marvel-Series, parameterName=Streak-G-U-Tile}
   B=ScalarParameter() {defaultValue=1, group=Marvel-Series, parameterName=Streak-G-V-Tile}
  )
  n17 = Add(
   A=Panner(Coordinate=Multiply(A=@n7, B=@n16), Time=@n8) {speedY=0.5}
   B=AppendVector(
    A=ScalarParameter() {defaultValue=1, group=Marvel-Series, parameterName=Streak-G-U-Offset}
    B=ScalarParameter() {defaultValue=1, group=Marvel-Series, parameterName=Streak-G-V-Offset}
   )
  )
  n18 = ScalarParameter() {group=Marvel-Series, parameterName=Streak-G-Power-Max}
  n19 = ScalarParameter() {group=Marvel-Series, parameterName=Streak-G-Power_Min}
  n20 = Multiply(
   A=Power(
    Base=TextureSample(UVs=@n17) {texture=&T_Marvel_Series}.G
    Exp=LinearInterpolate(A=@n18, B=@n19, Alpha=@n14)
   )
   B=LinearInterpolate(
    A=ScalarParameter() {group=Marvel-Series, parameterName=Streak-G-Multiply-Max}
    B=ScalarParameter() {group=Marvel-Series, parameterName=Streak-G-Multiply-Min}
    Alpha=@n14
   )
  )
  n21 = Saturate(None=Add(A=Saturate(None=@n15), B=Saturate(None=@n20)))
  n22 = Constant3Vector() {constant=<rgba 0.0027 0.058 0.462 1>}
  n23 = Constant2Vector() {g=0.5, r=0.5}
  n24 = ScalarParameter() {defaultValue=1.25, group=Marvel-Series, parameterName=Radius-Enhanced}
  n25 = ScalarParameter() {defaultValue=1.5, group=Marvel-Series, parameterName=Density-Enhanced}
  n26 = MaterialFunctionCall(Float2=TextureCoordinate() {coordinateIndex=3}) {materialFunction=&BreakOutFloat2Components}
  n27 = Divide(A=@n26.R, B=Max(A=@n26.G))
  n28 = Constant() {r=1}
  n29 = OneMinus(None=@n27)
  n30 = Add(
   A=Multiply(A=AppendVector(A=@n27, B=@n28), B=TextureCoordinate() {coordinateIndex=4})
   B=AppendVector(A=Multiply(A=@n29) {constB=0.5}, B=Constant())
  )
  n31 = Add(
   A=Multiply(A=LinearInterpolate(B=@n1, Alpha=@n21), B=ScalarParameter() {defaultValue=40, group=Marvel-Series, parameterName=Effect-Bump-Multiply})
   B=LinearInterpolate(
    A=FunctionInput(Preview=@n22) {inputName=RGB 2, sortPriority=1}
    B=@n1
    Alpha=MaterialFunctionCall(UVs=@n30, CenterPosition=@n23, Radius=@n24, Density=@n25) {materialFunction=&RadialGradientExponential}.RadialGradientExponential
   )
  )
  n32 = TextureSampleParameter2D(UVs=@n30) {group=Series, parameterName=SeriesTexture, texture=/Game/Global/Textures/Default/Masks_and_Gradients/T_Ui_LavaSeries_Bg}
  n33 = ScalarParameter() {parameterName=Background}
  FunctionOutput(None=StaticSwitchParameter(True=LinearInterpolate(A=@n31, B=@n32.RGB, Alpha=@n33), False=FunctionInput()) {parameterName=bIsMarvelSeries})
 ]}
