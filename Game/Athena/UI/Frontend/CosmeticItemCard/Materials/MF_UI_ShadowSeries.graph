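# Material function that renders the animated Shadow Series background for frontend cosmetic item cards, or passes an input through when bIsShadowSeries is off.
# An exponential radial two-color gradient blends with the series texture by a Background parameter.
# Two upward or downward panning noise and smoke layers tint it, using flowmap-offset, aspect-corrected UVs with a scalable screen-aligned offset.
# Desaturation and a color multiplier finish the result.
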MaterialFunction /Game/Athena/UI/Frontend/CosmeticItemCard/Materials/MF_UI_ShadowSeries
graph {nodes [
  FunctionInput(Preview=Constant3Vector() {constant=<rgba 0 0.4667 1 1>}) {inputName=RGB 1, sortPriority=1}
  FunctionInput(Preview=Constant3Vector() {constant=<rgba 0.0027 0.058 0.462 1>}) {inputName=RGB 2, sortPriority=1}
  n1 = Constant2Vector() {g=0.5, r=0.5}
  n2 = ScalarParameter() {defaultValue=0.75, group=Shadow-Series, parameterName=Radius-Enhanced}
  n3 = ScalarParameter() {defaultValue=2, group=Shadow-Series, parameterName=Density-Enhanced}
  n4 = MaterialFunctionCall(Float2=TextureCoordinate() {coordinateIndex=3}) {materialFunction=&BreakOutFloat2Components}
  n5 = Divide(A=@n4.R, B=Max(A=@n4.G))
  n6 = MaterialFunctionCall(Float2=TextureCoordinate() {coordinateIndex=4}) {materialFunction=&BreakOutFloat2Components}
  n7 = Multiply(A=AppendVector(A=@n5, B=Constant() {r=1}), B=AppendVector(A=@n6.R, B=OneMinus(None=@n6.G)))
  n8 = AppendVector(A=Multiply(A=OneMinus(None=@n5)) {constB=0.5}, B=Constant() {r=1})
  n9 = LinearInterpolate(
   A=LinearInterpolate(
    A=VectorParameter() {group=Shadow-Series, parameterName=RGB1, defaultValue=<rgba 0 0 0 1>}.RGB
    B=VectorParameter() {group=Shadow-Series, parameterName=RGB2, defaultValue=<rgba 0.1344 0.0708 0.401 1>}.RGB
    Alpha=MaterialFunctionCall(CenterPosition=@n1, Radius=@n2, Density=@n3) {materialFunction=&RadialGradientExponential}.RadialGradientExponential
   )
   B=TextureSampleParameter2D(UVs=Add(A=@n7, B=@n8)) {group=Series, parameterName=SeriesTexture, texture=/Game/Global/Textures/Default/Masks_and_Gradients/T_Ui_LavaSeries_Bg}.RGB
   Alpha=ScalarParameter() {group=Shadow-Series, parameterName=Background}
  )
  n10 = VectorParameter() {group=Shadow-Series, parameterName=Smoke-B-RGB, defaultValue=<rgba 0.0063 0.0049 0.0104 1>}
  n11 = TextureSample() {samplerType=SAMPLERTYPE_Normal, texture=/Game/Global/Textures/Default/Flowmaps/T_Smoke_FL_01}
  n12 = Multiply(
   A=Add(A=ComponentMask(None=@n11.RGB) {g=true, r=true}, B=TextureCoordinate() {coordinateIndex=4})
   B=ScalarParameter() {defaultValue=0.25, group=Shadow-Series, parameterName=UV Scale}
  ) {constB=0.5}
  n13 = MaterialFunctionCall(Float2=TextureCoordinate() {coordinateIndex=3}) {materialFunction=&BreakOutFloat2Components}
  n14 = Divide(A=@n13.R, B=Max(A=@n13.G))
  n15 = AppendVector(A=@n14, B=Constant() {r=1})
  n16 = Multiply(A=OneMinus(None=@n14)) {constB=0.5}
  n17 = Constant() {r=1}
  n18 = MaterialFunctionCall(Float2=MaterialFunctionCall() {materialFunction=&ScreenResolution}.Visible Resolution) {materialFunction=&BreakOutFloat2Components}
  n19 = AppendVector(A=Divide(A=@n18.R, B=Max(A=@n18.G)), B=Constant() {r=1})
  n20 = MaterialFunctionCall() {materialFunction=&ScreenAlignedUVs}
  n21 = Add(
   A=Add(A=Multiply(A=@n12, B=@n15), B=AppendVector(A=@n16, B=@n17))
   B=Multiply(A=Multiply(A=@n19, B=@n20."X 100%, Y 100%") {constB=0.5}, B=ScalarParameter())
  )
  n22 = Multiply(A=Time(), B=ScalarParameter() {defaultValue=1, group=Shadow-Series, parameterName=TimeFactor})
  n23 = Panner(Coordinate=Multiply(A=@n21, B=ScalarParameter() {group=Shadow-Series, parameterName=Smoke-B-Tiling-Min}), Time=@n22) {speedY=0.125}
  n24 = Saturate(None=Multiply(
    A=TextureSample(UVs=@n23) {samplerType=SAMPLERTYPE_LinearColor, texture=/Game/Characters/Player/Male/Medium/Bodies/M_MED_SRDrift/Textures/T_ElectricNoise}.R
    B=ScalarParameter() {defaultValue=1, group=Shadow-Series, parameterName=Smoke-B-Multiply}
   ))
  n25 = Multiply(A=@n21, B=ScalarParameter() {group=Shadow-Series, parameterName=Smoke-F-Tiling-Max})
  n26 = Multiply(
   A=TextureSample(UVs=Panner(Coordinate=@n25, Time=@n22) {speedY=0.25}) {texture=&T_Inky_Smoke_Tile}.R
   B=ScalarParameter() {defaultValue=1, group=Shadow-Series, parameterName=Smoke-F-Multiply}
  )
  n27 = Desaturation(
   None=LinearInterpolate(
    A=LinearInterpolate(A=@n9, B=@n10.RGB, Alpha=@n24)
    B=VectorParameter() {group=Shadow-Series, parameterName=Smoke-F-RGB, defaultValue=<rgba 0.0063 0.0049 0.0104 1>}.RGB
    Alpha=Saturate(None=@n26)
   )
   Fraction=ScalarParameter() {group=Shadow-Series, parameterName=Desat}
  )
  n28 = ScalarParameter() {defaultValue=1, group=Shadow-Series, parameterName=Color-Multiply}
  FunctionOutput(None=StaticSwitchParameter(True=Multiply(A=@n27, B=@n28), False=FunctionInput()) {parameterName=bIsShadowSeries})
 ]}
